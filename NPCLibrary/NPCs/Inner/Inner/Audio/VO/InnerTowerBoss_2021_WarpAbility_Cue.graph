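# Sound cue for the Inner Tower Boss NPC's warp ability voice line.
# A Random node picks among six equally weighted VO wave variants and is excluded from branch culling.
# Playback uses the shared close-range NPC VO attenuation and the critical enemy VO sound class.
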SoundCue /NPCLibrary/NPCs/Inner/Inner/Audio/VO/InnerTowerBoss_2021_WarpAbility_Cue
root=Random_0
nodes [
 {
  id=Random_0
  children [WavePlayer_0,WavePlayer_1,WavePlayer_2,WavePlayer_3,WavePlayer_4,WavePlayer_5]
  properties {bShouldExcludeFromBranchCulling=true, weights [1,1,1,1,1,1]}
 }
 {id=WavePlayer_0, properties {soundWaveAssetPtr=/NPCLibrary/NPCs/Inner/Inner/Audio/VO/Waves/InnerTowerBoss_2021_WarpAbility_031}}
 {id=WavePlayer_1, properties {soundWaveAssetPtr=/NPCLibrary/NPCs/Inner/Inner/Audio/VO/Waves/InnerTowerBoss_2021_WarpAbility_015}}
 {id=WavePlayer_2, properties {soundWaveAssetPtr=/NPCLibrary/NPCs/Inner/Inner/Audio/VO/Waves/InnerTowerBoss_2021_WarpAbility_014}}
 {id=WavePlayer_3, properties {soundWaveAssetPtr=/NPCLibrary/NPCs/Inner/Inner/Audio/VO/Waves/InnerTowerBoss_2021_WarpAbility_013}}
 {id=WavePlayer_4, properties {soundWaveAssetPtr=/NPCLibrary/NPCs/Inner/Inner/Audio/VO/Waves/InnerTowerBoss_2021_WarpAbility_012}}
 {id=WavePlayer_5, properties {soundWaveAssetPtr=/NPCLibrary/NPCs/Inner/Inner/Audio/VO/Waves/InnerTowerBoss_2021_WarpAbility_010}}
]
properties {
 attenuationSettings=/Game/Sounds/Fort_Human_VO_BR/NPCs/Shared/NPC_VO_Close_Att
 maxDistance=8500
 soundClassObject=&Fort_VO_Enemy_Critical
 volumeMultiplier=1
}
